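; Blueprint for a wharf wall light prop in the Apollo MountainBase environment, with optional time-of-day lighting.
; Its construction script calls the parent script, then checks the Use TOD Component flag.
; When enabled, it runs the TODLightingComponent's construct step using the Test Day Phase variable, presumably to preview lighting for that day phase.
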(fn ConstructionScript
  (|CallParentFunction)
  (if (Variables|Default|GetUse%20TOD%20Component)
    (Class|TODLightingComponent|RunConstruct (Variables|Default|GetTODLightingComponent) (Variables|Default|GetTest%20Day%20Phase))))
